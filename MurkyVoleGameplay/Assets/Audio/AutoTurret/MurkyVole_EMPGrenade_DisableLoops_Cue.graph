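# Looping Sound Cue for the EMP grenade disable effect in the AutoTurret audio, routed to the weapons sound class with its own attenuation.
# A Modulator feeds a Quality Level switch, whose first two slots randomly pick from five looping waves and last two from only the final three.
# The three-wave Random node is excluded from branch culling.
# A concurrency override allows one instance and prevents new ones, superseding the assigned explosion concurrency set.
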SoundCue /MurkyVoleGameplay/Assets/Audio/AutoTurret/MurkyVole_EMPGrenade_DisableLoops_Cue
root=Modulator_2
nodes [
 {id=Modulator_2, children [QualityLevel_0]}
 {id=QualityLevel_0, children [Random_3,Random_3,Random_8,Random_8]}
 {id=Random_3, children [WavePlayer_10, WavePlayer_11, WavePlayer_12, WavePlayer_13, WavePlayer_14], properties {weights [1,1,1,1,1]}}
 {id=Random_8, children [WavePlayer_12, WavePlayer_13, WavePlayer_14], properties {bShouldExcludeFromBranchCulling=true, weights [1,1,1]}}
 {id=WavePlayer_10, properties {bLooping=true, soundWaveAssetPtr=&/MurkyVole_EMP_DisableLoop_01}}
 {id=WavePlayer_11, properties {bLooping=true, soundWaveAssetPtr=&/MurkyVole_EMP_DisableLoop_02}}
 {id=WavePlayer_12, properties {bLooping=true, soundWaveAssetPtr=&/MurkyVole_EMP_DisableLoop_03}}
 {id=WavePlayer_13, properties {bLooping=true, soundWaveAssetPtr=&/MurkyVole_EMP_DisableLoop_04}}
 {id=WavePlayer_14, properties {bLooping=true, soundWaveAssetPtr=&/MurkyVole_EMP_DisableLoop_05}}
]
properties {
 attenuationSettings=/MurkyVoleGameplay/Assets/Audio/AutoTurret/Settings/MurkyVole_EMP_DisableLoops_Att
 bOverrideConcurrency=true
 maxDistance=6000
 soundClassObject=/Game/Sounds/FortSoundClassesAndModes/Fort_Weapons
 volumeMultiplier=0.8
 concurrencyOverrides {maxCount=1, resolutionRule=PreventNew, volumeScale=0.5}
 concurrencySet [/Game/Sounds/Concurrency_Rules/Weapon_Explosion_Conc]
}
