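# Sound cue for the impact when a MurkyVole auto turret drops and lands.
# A Random node picks one of four equally weighted DropLand waves and feeds it through a Modulator at the root.
# The cue lowers its volume, overrides attenuation for a short falloff range, and caps simultaneous instances by blocking new ones.
# It routes through the Fort_Loot_Collect sound class.
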SoundCue /MurkyVoleGameplay/Assets/Audio/AutoTurret/MurkyVole_Turret_DropImpact_Cue
root=Modulator_0
nodes [
 {id=Modulator_0, children [Random_1]}
 {id=Random_1, children [WavePlayer_4,WavePlayer_5,WavePlayer_6,WavePlayer_7], properties {weights [1,1,1,1]}}
 {id=WavePlayer_4, properties {soundWaveAssetPtr=&/MurkyVole_AutoTurret_DropLand_01}}
 {id=WavePlayer_5, properties {soundWaveAssetPtr=&/MurkyVole_AutoTurret_DropLand_02}}
 {id=WavePlayer_6, properties {soundWaveAssetPtr=&/MurkyVole_AutoTurret_DropLand_03}}
 {id=WavePlayer_7, properties {soundWaveAssetPtr=&/MurkyVole_AutoTurret_DropLand_04}}
]
properties {
 bOverrideAttenuation=true
 bOverrideConcurrency=true
 maxDistance=1000
 soundClassObject=&Fort_Loot_Collect
 volumeMultiplier=0.7
 attenuationOverrides {falloffDistance=1000, attenuationShapeExtents {x=0}}
 concurrencyOverrides {maxCount=3, resolutionRule=PreventNew}
}
