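# Sound cue for the lift-down sound of the MurkyVole auto turret, classed as a gadget item sound.
# A Random node picks one of three equally weighted LiftDown wave variations and feeds it through a Modulator at the root.
# Playback uses the turret ambient attenuation at reduced volume, with a concurrency override that allows one instance and stops the oldest.
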SoundCue /MurkyVoleGameplay/Assets/Audio/AutoTurret/MurkyVole_Turret_LiftDown_Cue
root=Modulator_2
nodes [
 {id=Modulator_2, children [Random_4]}
 {id=Random_4, children [WavePlayer_12, WavePlayer_13, WavePlayer_14], properties {bShouldExcludeFromBranchCulling=true, weights [1,1,1]}}
 {id=WavePlayer_12, properties {soundWaveAssetPtr=&/MurkyVole_Turret_Fire_LiftDown_03}}
 {id=WavePlayer_13, properties {soundWaveAssetPtr=&/MurkyVole_Turret_Fire_LiftDown_02}}
 {id=WavePlayer_14, properties {soundWaveAssetPtr=&/MurkyVole_Turret_Fire_LiftDown_01}}
]
properties {
 attenuationSettings=/MurkyVoleGameplay/Assets/Audio/AutoTurret/Settings/MurkyVole_Turret_Ambient_Att
 bOverrideConcurrency=true
 maxDistance=1050
 soundClassObject=&Fort_Gadgets_Items
 volumeMultiplier=0.4
 concurrencyOverrides {maxCount=1, resolutionRule=StopOldest}
}
