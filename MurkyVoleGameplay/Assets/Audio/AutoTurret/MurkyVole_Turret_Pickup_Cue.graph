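# Pickup sound cue for the MurkyVole auto turret, using the loot-collect sound class and weapon-pickup attenuation.
# A mixer layers the turret pickup wave with a louder hover-jet pickup wave borrowed from the JetPack audio folder.
# The root mixer blends that layer at reduced volume with a delayed, randomly chosen turret equip sound at full volume.
# Concurrency settings restrict overlapping pickup playback.
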SoundCue /MurkyVoleGameplay/Assets/Audio/AutoTurret/MurkyVole_Turret_Pickup_Cue
root=Mixer_1
nodes [
 {id=Delay_1, children [Modulator_2], properties {delayMax=0.5, delayMin=0.5}}
 {id=Mixer_0, children [Modulator_1,WavePlayer_7], properties {inputVolume [1,2]}}
 {id=Mixer_1, children [Mixer_0,Delay_1], properties {inputVolume [0.5,1]}}
 {id=Modulator_1, children [WavePlayer_0], properties {pitchMax=1, pitchMin=1}}
 {id=Modulator_2, children [Random_2], properties {pitchMax=1, pitchMin=1, volumeMax=1, volumeMin=1}}
 {id=Random_2, children [WavePlayer_1,WavePlayer_2,WavePlayer_5,WavePlayer_6], properties {weights [1,1,1,1]}}
 {id=WavePlayer_0, properties {soundWaveAssetPtr=&/MurkyVole_AutoTurret_PickUp}}
 {id=WavePlayer_1, properties {soundWaveAssetPtr=&/MurkyVole_AutoTurret_Equip_01}}
 {id=WavePlayer_2, properties {soundWaveAssetPtr=&/MurkyVole_AutoTurret_Equip_02}}
 {id=WavePlayer_5, properties {soundWaveAssetPtr=&/MurkyVole_AutoTurret_Equip_05}}
 {id=WavePlayer_6, properties {soundWaveAssetPtr=&/MurkyVole_AutoTurret_Equip_06}}
 {id=WavePlayer_7, properties {soundWaveAssetPtr=/MurkyVoleGameplay/Assets/Audio/JetPack/MurkyVole_HoverJets_PickUp_01}}
]
properties {
 attenuationSettings=&Fort_Weapon_Pickup_Att
 maxDistance=3500
 soundClassObject=&Fort_Loot_Collect
 volumeMultiplier=0.5
 concurrencyOverrides {maxCount=1, resolutionRule=PreventNew}
 concurrencySet [&PickupMaximum_Oldest]
}
