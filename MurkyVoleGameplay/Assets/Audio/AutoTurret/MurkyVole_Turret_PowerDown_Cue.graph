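# Sound cue for the MurkyVole auto turret's power-down, playing a randomly chosen LaserOff sound through a Modulator for variation.
# A quality-level switch gives the first two levels all six variants and the last two only variants 04 to 06.
# It uses the turret Sight attenuation asset at half volume.
# Overridden concurrency allows one instance per owner and blocks new plays while one is active.
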SoundCue /MurkyVoleGameplay/Assets/Audio/AutoTurret/MurkyVole_Turret_PowerDown_Cue
root=Modulator_1
nodes [
 {id=Modulator_1, children [QualityLevel_0]}
 {id=QualityLevel_0, children [Random_1,Random_1,Random_2,Random_2]}
 {id=Random_1, children [WavePlayer_1,WavePlayer_4,WavePlayer_5,WavePlayer_6,WavePlayer_7,WavePlayer_8], properties {weights [1,1,1,1,1,1]}}
 {id=Random_2, children [WavePlayer_6,WavePlayer_7,WavePlayer_8], properties {bShouldExcludeFromBranchCulling=true, weights [1,1,1]}}
 {id=WavePlayer_1, properties {soundWaveAssetPtr=&/MurkyVole_AutoTurret_LaserOff_01}}
 {id=WavePlayer_4, properties {soundWaveAssetPtr=&/MurkyVole_AutoTurret_LaserOff_02}}
 {id=WavePlayer_5, properties {soundWaveAssetPtr=&/MurkyVole_AutoTurret_LaserOff_03}}
 {id=WavePlayer_6, properties {soundWaveAssetPtr=&/MurkyVole_AutoTurret_LaserOff_04}}
 {id=WavePlayer_7, properties {soundWaveAssetPtr=&/MurkyVole_AutoTurret_LaserOff_05}}
 {id=WavePlayer_8, properties {soundWaveAssetPtr=&/MurkyVole_AutoTurret_LaserOff_06}}
]
properties {
 attenuationSettings=/MurkyVoleGameplay/Assets/Audio/AutoTurret/Settings/MurkyVole_Turret_Sight_Att
 bOverrideConcurrency=true
 maxDistance=4100
 soundClassObject=&Fort_Gadgets_Items
 volumeMultiplier=0.5
 attenuationOverrides {
  bEnableFocusInterpolation=true
  bEnableListenerFocus=true
  falloffDistance=1200
  nonFocusVolumeAttenuation=0.5
  stereoSpread=500
  attenuationShapeExtents {x=0}
 }
 concurrencyOverrides {bLimitToOwner=true, maxCount=1, resolutionRule=PreventNew}
}
