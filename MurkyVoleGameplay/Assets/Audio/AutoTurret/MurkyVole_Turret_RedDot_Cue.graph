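# Sound cue for the MurkyVole auto turret's red dot, apparently a cue that the turret has targeted something.
# It plays the single Targeted wave through a Modulator with pitch and volume fixed at unity, with a raised cue volume multiplier.
# A dedicated turret target attenuation asset spatializes it within the gadgets and items sound class.
# Its own concurrency rule allows one instance per owner and blocks new plays while one is active.
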SoundCue /MurkyVoleGameplay/Assets/Audio/AutoTurret/MurkyVole_Turret_RedDot_Cue
root=Modulator_1
nodes [
 {id=Modulator_1, children [WavePlayer_0], properties {pitchMax=1, pitchMin=1, volumeMax=1, volumeMin=1}}
 {id=WavePlayer_0, properties {soundWaveAssetPtr=&/MurkyVole_AutoTurret_Targeted}}
]
properties {
 attenuationSettings=/MurkyVoleGameplay/Assets/Audio/AutoTurret/Settings/MurkyVole_Turret_Target_Att
 bOverrideConcurrency=true
 maxDistance=4100
 soundClassObject=&Fort_Gadgets_Items
 volumeMultiplier=1.5
 attenuationOverrides {
  bEnableFocusInterpolation=true
  bEnableListenerFocus=true
  falloffDistance=1200
  nonFocusVolumeAttenuation=0.5
  stereoSpread=500
  attenuationShapeExtents {x=0}
 }
 concurrencyOverrides {bLimitToOwner=true, maxCount=1, resolutionRule=PreventNew}
}
