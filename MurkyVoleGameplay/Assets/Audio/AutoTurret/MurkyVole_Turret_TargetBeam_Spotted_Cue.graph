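# Sound cue for the MurkyVole auto turret's target beam spotted event, playing a laser-on sound from a set of variations.
# A Modulator feeds a quality switch whose first two levels pick randomly from six waves and whose last two pick from a three-wave subset excluded from branch culling.
# It is spatialized with temporary turret sight attenuation and limited to one instance per owner, preventing new plays.
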SoundCue /MurkyVoleGameplay/Assets/Audio/AutoTurret/MurkyVole_Turret_TargetBeam_Spotted_Cue
root=Modulator_0
nodes [
 {id=Modulator_0, children [QualityLevel_0]}
 {id=QualityLevel_0, children [Random_1,Random_1,Random_2,Random_2]}
 {id=Random_1, children [WavePlayer_0,WavePlayer_1,WavePlayer_2,WavePlayer_3,WavePlayer_5,WavePlayer_7], properties {weights [1,1,1,1,1,1]}}
 {id=Random_2, children [WavePlayer_3,WavePlayer_5,WavePlayer_7], properties {bShouldExcludeFromBranchCulling=true, weights [1,1,1]}}
 {id=WavePlayer_0, properties {soundWaveAssetPtr=&/MurkyVole_AutoTurret_LaserOn_01}}
 {id=WavePlayer_1, properties {soundWaveAssetPtr=&/MurkyVole_AutoTurret_LaserOn_02}}
 {id=WavePlayer_2, properties {soundWaveAssetPtr=&/MurkyVole_AutoTurret_LaserOn_03}}
 {id=WavePlayer_3, properties {soundWaveAssetPtr=&/MurkyVole_AutoTurret_LaserOn_04}}
 {id=WavePlayer_5, properties {soundWaveAssetPtr=&/MurkyVole_AutoTurret_LaserOn_05}}
 {id=WavePlayer_7, properties {soundWaveAssetPtr=&/MurkyVole_AutoTurret_LaserOn_06}}
]
properties {
 attenuationSettings=/MurkyVoleGameplay/Assets/Turret/Temp/Settings/TEMP_MurkyVole_Turret_Sight_Att
 bOverrideConcurrency=true
 maxDistance=4100
 soundClassObject=&Fort_Gadgets_Items
 volumeMultiplier=1
 attenuationOverrides {
  bEnableFocusInterpolation=true
  bEnableListenerFocus=true
  falloffDistance=1200
  nonFocusVolumeAttenuation=0.5
  stereoSpread=500
  attenuationShapeExtents {x=0}
 }
 concurrencyOverrides {bLimitToOwner=true, maxCount=1, resolutionRule=PreventNew}
}
