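# Sound cue for the MurkyVole auto turret's toggle-on sound, in the gadgets and items sound class.
# A mixer layers the main toggle-on wave, passed through a modulator, with one of three equally weighted random glitch waves.
# It uses the turret's shared ambient attenuation asset.
# Overridden concurrency allows one instance per owner and blocks new plays while one is active.
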SoundCue /MurkyVoleGameplay/Assets/Audio/AutoTurret/MurkyVole_Turret_Toggle_On_Cue
root=Mixer_0
nodes [
 {id=Mixer_0, children [Modulator_0,Random_0], properties {inputVolume [1,1]}}
 {id=Modulator_0, children [WavePlayer_0]}
 {id=Random_0, children [WavePlayer_1,WavePlayer_2,WavePlayer_3], properties {weights [1,1,1]}}
 {id=WavePlayer_0, properties {soundWaveAssetPtr=&/MurkyVole_AutoTurret_Toggle_On}}
 {id=WavePlayer_1, properties {soundWaveAssetPtr=&/MurkyVole_AutoTurret_Toggle_On_Glitch_01}}
 {id=WavePlayer_2, properties {soundWaveAssetPtr=&/MurkyVole_AutoTurret_Toggle_On_Glitch_02}}
 {id=WavePlayer_3, properties {soundWaveAssetPtr=&/MurkyVole_AutoTurret_Toggle_On_Glitch_03}}
]
properties {
 attenuationSettings=/MurkyVoleGameplay/Assets/Audio/AutoTurret/Settings/MurkyVole_Turret_Ambient_Att
 bOverrideConcurrency=true
 maxDistance=1050
 soundClassObject=&Fort_Gadgets_Items
 volumeMultiplier=0.7
 attenuationOverrides {
  bEnableFocusInterpolation=true
  bEnableListenerFocus=true
  falloffDistance=1200
  nonFocusVolumeAttenuation=0.5
  stereoSpread=500
  attenuationShapeExtents {x=0}
 }
 concurrencyOverrides {bLimitToOwner=true, maxCount=1, resolutionRule=PreventNew}
}
